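# Masked, two-sided lit particle material for glinting ice flakes on Elemental weapon skins.
# Opacity comes from a flare texture, with an optional random channel pick, through a distance fade.
# Fake tumbling reflections rotate around random per-particle axes by distance from world origin, so movement drives spin.
# The reflection vector samples a panned glass texture, which is sharpened, tinted, multiplied by a radial two-color gradient and particle alpha for emissive.
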Material /Game/Weapons/WeaponSkins/Elemental/Materials/M_Elemental_Ice_Flakes_01
graph {
 properties {
  bEnableResponsiveAA=true
  bUsedWithNiagaraSprites=true
  bUsedWithParticleSprites=true
  blendMode=BLEND_Masked
  opacityMaskClipValue=0.1
  pixelDepthOffsetMode=PDOM_Legacy
  translucencyLightingMode=TLM_VolumetricPerVertexDirectional
  twoSided=true
  usedShadingModels=Default Lit
 }
 nodes [
  n1 = ParticleRandom()
  n2 = TextureSampleParameter2D() {parameterName=Opacity, texture=/Game/Effects/Fort_Effects/Textures/Flares/T_Toon_Flare1}
  MaterialFunctionCall(
   Opacity In=StaticSwitchParameter(
    True=MaterialFunctionCall!(Random Picker=@n1, Texture Input=@n2.RGB, Option 1, Option 2, Option 3) {materialFunction=/Game/Effects/Fort_Effects/Materials/_Functions/MF_RGB_Random_01}.Result
    False=@n2.RGB
   ) {parameterName=Use_RGB_Random}
   Position=ParticlePositionWS()
   Fade_Start_Distance=ScalarParameter() {defaultValue=120, parameterName=Start Scale}
   Fade_End_Distance=ScalarParameter() {defaultValue=100, parameterName=End Scale}
  ) {materialFunction=&MF_Distance_SyncedFade_FX}
  n3 = Normalize(VectorInput=MaterialFunctionCall(
    Random Picker=ParticleRandom()
    Option 1=VectorParameter() {parameterName=NormalizeRotationAxis_1, defaultValue=<rgba 1 0.3 -0.2 0>}.RGB
    Option 2=VectorParameter() {parameterName=NormalizeRotationAxis_2, defaultValue=<rgba -0.44 1 0.13 0>}.RGB
    Option 3=VectorParameter() {parameterName=NormalizeRotationAxis_3, defaultValue=<rgba 0.06 -0.22 1 0>}.RGB
   ) {materialFunction=&MF_RGB_Random_01}.Vec3 Out Only)
  n4 = Constant3Vector()
  n5 = ParticlePositionWS()
  n6 = Multiply(
   A=Divide(A=Distance(A=@n4, B=@n5), B=ScalarParameter() {defaultValue=1.8, parameterName=Spin Scale})
   B=Clamp(None=ParticleRandom()) {minDefault=0.55}
  )
  n7 = ParticlePositionWS()
  n8 = WorldPosition()
  n9 = Add(
   A=ReflectionVectorWS(CustomWorldNormal=RotateAboutAxis(NormalizedRotationAxis=@n3, RotationAngle=@n6, PivotPoint=@n7, Position=@n8.XYZ)) {bNormalizeCustomWorldNormal=true}
   B=ConstantBiasScale(None=ParticleRandom()) {bias=-0.5, scale=2}
  )
  n10 = TextureSample(UVs=Panner(
    Coordinate=ComponentMask(None=@n9) {g=true, r=true}
    Speed=VectorParameter() {parameterName=Panner Speed, defaultValue=<rgba 0.45 0.2 0 0>}.RGB
   ) {speedX=1, speedY=1}) {texture=/Game/Items/Art_noLOD/Textures/T_Items_MineralPowder_Glass_Reflection}
  n11 = ScalarParameter() {defaultValue=4, parameterName=CubemapExponent}
  n12 = Add(
   A=Multiply(A=Power(Base=@n10.RGB, Exp=@n11), B=VectorParameter() {parameterName=Cubemap Color, defaultValue=<rgba 7.8125 96.3956 150 0>}.RGB)
   B=ScalarParameter() {defaultValue=0.4, parameterName=CubemapBias}
  )
  n13 = LinearInterpolate(
   A=VectorParameter() {parameterName=Color Outer, defaultValue=<rgba 0.4887 0.5995 0.905 1>}.RGB
   B=VectorParameter() {parameterName=Color Inner, defaultValue=<rgba 0.385 0.7314 1 1>}.RGB
   Alpha=SphereMask(
    A=TextureCoordinate()
    B=Constant() {r=0.5}
    Radius=ScalarParameter() {defaultValue=0.45, parameterName=Radius}
    Hardness=ScalarParameter() {defaultValue=0.2, parameterName=Hardness}
   )
  )
  MaterialFunctionCall(Emissive=Multiply(A=Multiply(A=@n12, B=@n13) {constB=1.5}, B=ParticleColor().A)) {materialFunction=&MF_FX_EmissiveAdjustments}
  ScalarParameter() {defaultValue=0.8, parameterName=Metallic}
  ScalarParameter() {defaultValue=1, parameterName=Specular}
  ScalarParameter() {defaultValue=0.15, parameterName=Roughness}
 ]
}
