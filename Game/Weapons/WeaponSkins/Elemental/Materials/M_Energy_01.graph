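# Unlit, two-sided translucent energy material for elemental weapon skin particles, supporting sprites and meshes.
# Two panning storm-swirl samples, warped by sine-pulsed flowmaps with dynamic-parameter tiling and speed, form noise that particle alpha erodes.
# Rounded noise thresholds pick three banded blue colors for emissive, while opacity combines noise, sphere or vertex-color mask, Fresnel and depth fade.
# The summed swirl samples also drive world-position offset scaled by vertex normal.
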Material /Game/Weapons/WeaponSkins/Elemental/Materials/M_Energy_01
graph {
 properties {
  bEnableResponsiveAA=true
  bTangentSpaceNormal=false
  bUsedWithMeshParticles=true
  bUsedWithNiagaraMeshParticles=true
  bUsedWithParticleSprites=true
  blendMode=BLEND_Translucent
  opacityMaskClipValue=0.35
  pixelDepthOffsetMode=PDOM_Legacy
  shadingModel=MSM_Unlit
  twoSided=true
  usedShadingModels=Unlit
 }
 nodes [
  n1 = DynamicParameter() {paramNames [Panner Speed, U Tile, V Tile, Speed]}
  n2 = AppendVector(A=@n1.U Tile, B=@n1.U Tile)
  n3 = Multiply(A=TextureCoordinate() {uTiling=0.25, vTiling=0.3484}, B=@n2)
  n4 = Sine(None=Time()) {period=5.721}
  n5 = Multiply(
   A=ComponentMask(None=TextureSample() {samplerType=SAMPLERTYPE_Normal, texture=/Game/Global/Textures/Default/Flowmaps/T_Smoke_FL_01}.RGB) {g=true, r=true}
   B=LinearInterpolate(B=ScalarParameter() {defaultValue=0.1115, parameterName=FL A}, Alpha=ConstantBiasScale(None=@n4) {bias=3, scale=0.25})
  )
  n6 = Multiply(A=Time(), B=DynamicParameter() {paramNames [Panner Speed, U Tile, V Tile, Speed]}.Panner Speed)
  n7 = TextureSample(UVs=Panner(Coordinate=Add(A=@n3, B=@n5), Time=@n6) {bFractionalPart=true, speedX=0.0653, speedY=0.2066}) {samplerType=SAMPLERTYPE_Normal, texture=/Game/VisualThreatManager/StormVisuals/LargeClouds/T_StormSwirlNormal}
  n8 = Multiply(A=@n2, B=TextureCoordinate() {uTiling=0.5, vTiling=0.2062})
  n9 = TextureCoordinate() {uTiling=-1, vTiling=-1}
  n10 = Sine(None=Time()) {period=6.622}
  n11 = Multiply(
   A=ComponentMask(None=TextureSample(UVs=@n9) {samplerType=SAMPLERTYPE_Normal, texture=&T_Smoke_FL_01}.RGB) {g=true, r=true}
   B=LinearInterpolate(B=ScalarParameter() {defaultValue=0.089, parameterName=FL B}, Alpha=ConstantBiasScale(None=@n10) {bias=3, scale=0.25})
  )
  n12 = TextureSample(UVs=Panner(Coordinate=Add(A=@n8, B=@n11), Time=@n6) {bFractionalPart=true, speedX=-0.0617, speedY=0.108}) {samplerType=SAMPLERTYPE_Normal, texture=&T_StormSwirlNormal}
  n13 = Add(A=Divide(A=Add(A=@n7.R, B=@n12.R)), B=OneMinus(None=ParticleColor().A))
  n14 = Add(A=OneMinus(None=ConstantBiasScale(None=@n13) {bias=0.5, scale=1.5}), B=ScalarParameter() {defaultValue=0.2, parameterName=Opacity Bias})
  n15 = TextureCoordinate()
  n16 = Constant() {r=0.5}
  n17 = ScalarParameter() {defaultValue=0.5, parameterName=Radius}
  n18 = ScalarParameter() {defaultValue=0.3, parameterName=Hardness}
  n19 = OneMinus(None=StaticSwitchParameter(True=SphereMask(A=@n15, B=@n16, Radius=@n17, Hardness=@n18), False=VertexColor().B) {defaultValue=true, parameterName=Use Spheremask})
  n20 = ScalarParameter() {defaultValue=2, parameterName=Fresnel Brightness}
  n21 = ScalarParameter() {defaultValue=1.5, parameterName=Fresnel Exponent}
  n22 = VertexNormalWS()
  DepthFade(
   Opacity=Multiply(
    A=Subtract(A=@n14, B=@n19)
    B=MaterialFunctionCall(Brightness=@n20, Exponent=@n21, Vertex / Pixel Normal=@n22) {materialFunction=&MF_Fresnel_FX}.Fresnel Double Sided
   )
   FadeDistance=ScalarParameter() {defaultValue=10, parameterName=Fade Distance}
  )
  n23 = VectorParameter() {parameterName=Color A, defaultValue=<rgba 1.6 7.3526 20 0>}
  n24 = VectorParameter() {parameterName=Color B, defaultValue=<rgba 0.0188 0.3469 1.5 0>}
  n25 = Saturate(None=Round(None=ConstantBiasScale(None=@n13) {bias=0.5, scale=2.5}))
  n26 = Round(None=ConstantBiasScale(None=@n13) {bias=0.5, scale=2})
  MaterialFunctionCall(Emissive=LinearInterpolate(
    A=LinearInterpolate(A=@n23.RGB, B=@n24.RGB, Alpha=@n25)
    B=VectorParameter() {parameterName=Color C, defaultValue=<rgba 0 0.0993 0.535 1>}.RGB
    Alpha=Saturate(None=@n26)
   )) {materialFunction=&MF_FX_EmissiveAdjustments}
  n27 = Add(A=@n7.RGB, B=@n12.RGB)
  Multiply(
   A=Multiply(A=Normalize(VectorInput=@n27), B=VertexNormalWS())
   B=VectorParameter() {parameterName=WPO Intensity, defaultValue=<rgba 5 5 0 0>}.RGB
  )
 ]
}
